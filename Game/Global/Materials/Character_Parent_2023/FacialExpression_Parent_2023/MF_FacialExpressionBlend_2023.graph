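# Material function for the character parent materials that blends facial expression textures over the incoming Material attributes, behind the UseFacialExpressionBlend static switch.
# Static switches gate lerps toward expression Normal, Diffuse and SRM textures, with only normals blended by default.
# The saturated Normal_Curve scalar is declared as the Blend weight.
# As wired, SRM input repeats Specular in all channels, and its R output feeds Specular, Roughness and Metallic.
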MaterialFunction /Game/Global/Materials/Character_Parent_2023/FacialExpression_Parent_2023/MF_FacialExpressionBlend_2023
graph {nodes [
  n1 = MaterialFunctionCall(Float3=NamedRerouteUsage()) {materialFunction=&BreakOutFloat3Components}
  n2 = SetMaterialAttributes!(
   MaterialAttributes=NamedRerouteUsage()
   BaseColor=NamedRerouteUsage()
   Normal=NamedRerouteUsage()
   Specular=@n1.R
   Roughness=@n1.R
   Metallic=@n1.R
  ) {attributeSetTypes [
    #1
    #2
    #3
    #4
    #5
   ]}
  FunctionOutput(None=StaticSwitchParameter(True=@n2, False=NamedRerouteUsage()) {defaultValue=true, group=FacialExpressionBlend, parameterName=UseFacialExpressionBlend, sortPriority=0})
  n3 = GetMaterialAttributes(None=NamedRerouteDeclaration(None=FunctionInput() {inputName=Material, inputType=FunctionInput_MaterialAttributes}) {name=IN_Material, nodeColor=<rgba 1 0.0234 0 1>}) {attributeGetTypes [
    #1
    #2
    #3
    #4
    #5
   ]}
  NamedRerouteDeclaration(None=@n3.BaseColor) {name=IN_Diffuse, nodeColor=<rgba 0 0.2969 1 1>}
  NamedRerouteDeclaration(None=@n3.Normal) {name=IN_Normal, nodeColor=<rgba 0 0.2969 1 1>}
  NamedRerouteDeclaration(None=MaterialFunctionCall(X=@n3.Specular, Y=@n3.Specular, Z=@n3.Specular) {materialFunction=&MakeFloat3}.Result) {name=IN_SRM, nodeColor=<rgba 0 0.2969 1 1>}
  n4 = NamedRerouteUsage()
  n5 = TextureSampleParameter2D() {
   group=FacialExpressionBlend
   parameterName=FacialExpression_Normals
   samplerType=SAMPLERTYPE_Normal
   sortPriority=11
   texture=&FlatNormalBlue
  }
  n6 = NamedRerouteUsage()
  NamedRerouteDeclaration(None=StaticSwitchParameter(True=LinearInterpolate(A=@n4, B=@n5.RGB, Alpha=@n6), False=NamedRerouteUsage()) {defaultValue=true, group=FacialExpressionBlend, parameterName=BlendNormals, sortPriority=10}) {name=OUT_Normal, nodeColor=<rgba 1 0 0.7969 1>}
  n7 = NamedRerouteUsage()
  n8 = TextureSampleParameter2D() {
   group=FacialExpressionBlend
   parameterName=FacialExpression_Diffuse
   sortPriority=21
   texture=/Game/Global/Textures/Default/Blanks/T_MidgreysRGB
  }
  n9 = NamedRerouteUsage()
  NamedRerouteDeclaration(None=StaticSwitchParameter(True=LinearInterpolate(A=@n7, B=@n8.RGB, Alpha=@n9), False=NamedRerouteUsage()) {group=FacialExpressionBlend, parameterName=BlendDiffuse, sortPriority=20}) {name=OUT_Diffuse, nodeColor=<rgba 1 0 0.7969 1>}
  NamedRerouteDeclaration(None=Saturate(None=ScalarParameter() {group=FacialExpressionBlend, parameterName=Normal_Curve, sortPriority=1})) {name=Blend, nodeColor=<rgba 0 0.4609 1 1>}
  n10 = NamedRerouteUsage()
  n11 = TextureSampleParameter2D() {
   group=FacialExpressionBlend
   parameterName=FacialExpression_SRM
   samplerType=SAMPLERTYPE_Masks
   sortPriority=31
   texture=&T_Fortnite_Default_S
  }
  n12 = NamedRerouteUsage()
  NamedRerouteDeclaration(None=StaticSwitchParameter(True=LinearInterpolate(A=@n10, B=@n11.RGB, Alpha=@n12), False=NamedRerouteUsage()) {group=FacialExpressionBlend, parameterName=BlendSRM, sortPriority=30}) {name=OUT_SRM, nodeColor=<rgba 1 0 0.7969 1>}
 ]}
