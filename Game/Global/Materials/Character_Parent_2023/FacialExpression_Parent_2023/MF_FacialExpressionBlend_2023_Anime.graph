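# Material function for anime-style characters that blends facial expression textures over the base face maps: STT, Normal, DFL, SRM, Color_Lit and Color_Shaded.
# A saturated Normal_Curve scalar provides the blend weight for linear interpolations between each input and its expression texture.
# Static switches gate each map, with both colors sharing BlendDiffuse and only normals blended by default.
# A master UseFacialExpressionBlend switch, on by default, selects each output.
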MaterialFunction /Game/Global/Materials/Character_Parent_2023/FacialExpression_Parent_2023/MF_FacialExpressionBlend_2023_Anime
graph {nodes [
  FunctionInput() {inputName=--------, inputType=FunctionInput_Bool, sortPriority=5}
  FunctionOutput(None=StaticSwitchParameter(True=NamedRerouteUsage(), False=NamedRerouteUsage()) {defaultValue=true, group=FacialExpressionBlend, parameterName=UseFacialExpressionBlend, sortPriority=0}) {outputName=STT, sortPriority=1}
  FunctionOutput(None=StaticSwitchParameter(True=NamedRerouteUsage(), False=NamedRerouteUsage()) {defaultValue=true, group=FacialExpressionBlend, parameterName=UseFacialExpressionBlend, sortPriority=0}) {outputName=Normal, sortPriority=2}
  FunctionOutput(None=StaticSwitchParameter(True=NamedRerouteUsage(), False=NamedRerouteUsage()) {defaultValue=true, group=FacialExpressionBlend, parameterName=UseFacialExpressionBlend, sortPriority=0}) {outputName=SRM, sortPriority=8}
  FunctionOutput(None=StaticSwitchParameter(True=NamedRerouteUsage(), False=NamedRerouteUsage()) {defaultValue=true, group=FacialExpressionBlend, parameterName=UseFacialExpressionBlend, sortPriority=0}) {outputName=Color_Lit, sortPriority=6}
  FunctionOutput(None=StaticSwitchParameter(True=NamedRerouteUsage(), False=NamedRerouteUsage()) {defaultValue=true, group=FacialExpressionBlend, parameterName=UseFacialExpressionBlend, sortPriority=0}) {outputName=Color_Shaded, sortPriority=7}
  FunctionOutput() {outputName=--------, sortPriority=5}
  FunctionOutput(None=StaticSwitchParameter(True=NamedRerouteUsage(), False=NamedRerouteUsage()) {defaultValue=true, group=FacialExpressionBlend, parameterName=UseFacialExpressionBlend, sortPriority=0}) {outputName=DFL, sortPriority=3}
  NamedRerouteDeclaration(None=FunctionInput() {inputName=STT, sortPriority=1}) {name=IN_STT, nodeColor=<rgba 0 0.2969 1 1>}
  NamedRerouteDeclaration(None=FunctionInput() {inputName=Color_Lit, sortPriority=6}) {name=IN_DiffuseLit, nodeColor=<rgba 0 0.2969 1 1>}
  n1 = NamedRerouteUsage()
  n2 = TextureSampleParameter2D() {
   group=FacialExpressionBlend
   parameterName=FacialExpression_DFL
   samplerType=SAMPLERTYPE_Alpha
   sortPriority=51
   texture=/Game/Global/Textures/Default/Blanks/T_BlackAlpha
  }
  n3 = NamedRerouteUsage()
  NamedRerouteDeclaration(None=StaticSwitchParameter(True=LinearInterpolate(A=@n1, B=@n2.RGB, Alpha=@n3), False=NamedRerouteUsage()) {group=FacialExpressionBlend, parameterName=BlendDFL, sortPriority=50}) {name=OUT_DFL, nodeColor=<rgba 1 0 0.7969 1>}
  n4 = NamedRerouteUsage()
  n5 = TextureSampleParameter2D() {
   group=FacialExpressionBlend
   parameterName=FacialExpression_ColorShaded
   sortPriority=22
   texture=/Game/Global/Textures/Default/Blanks/T_MidgreysRGB
  }
  n6 = NamedRerouteUsage()
  NamedRerouteDeclaration(None=StaticSwitchParameter(True=LinearInterpolate(A=@n4, B=@n5.RGB, Alpha=@n6), False=NamedRerouteUsage()) {group=FacialExpressionBlend, parameterName=BlendDiffuse, sortPriority=20}) {name=OUT_DiffuseShaded, nodeColor=<rgba 0.4531 0 1 1>}
  NamedRerouteDeclaration(None=FunctionInput() {inputName=DFL, inputType=FunctionInput_Scalar, sortPriority=3}) {name=IN_DFL, nodeColor=<rgba 0 0.2969 1 1>}
  NamedRerouteDeclaration(None=FunctionInput() {inputName=Normal, sortPriority=2}) {name=IN_Normal, nodeColor=<rgba 0 0.2969 1 1>}
  NamedRerouteDeclaration(None=FunctionInput() {inputName=SRM, sortPriority=8}) {name=IN_SRM, nodeColor=<rgba 0 0.2969 1 1>}
  n7 = NamedRerouteUsage()
  n8 = TextureSampleParameter2D() {
   group=FacialExpressionBlend
   parameterName=FacialExpression_Normals
   samplerType=SAMPLERTYPE_Normal
   sortPriority=11
   texture=&FlatNormalBlue
  }
  n9 = NamedRerouteUsage()
  NamedRerouteDeclaration(None=StaticSwitchParameter(True=LinearInterpolate(A=@n7, B=@n8.RGB, Alpha=@n9), False=NamedRerouteUsage()) {defaultValue=true, group=FacialExpressionBlend, parameterName=BlendNormals, sortPriority=10}) {name=OUT_Normal, nodeColor=<rgba 1 0 0.7969 1>}
  n10 = NamedRerouteUsage()
  n11 = TextureSampleParameter2D() {
   group=FacialExpressionBlend
   parameterName=FacialExpression_ColorLit
   sortPriority=21
   texture=&T_MidgreysRGB
  }
  n12 = NamedRerouteUsage()
  NamedRerouteDeclaration(None=StaticSwitchParameter(True=LinearInterpolate(A=@n10, B=@n11.RGB, Alpha=@n12), False=NamedRerouteUsage()) {group=FacialExpressionBlend, parameterName=BlendDiffuse, sortPriority=20}) {name=OUT_Diffuse, nodeColor=<rgba 1 0 0.7969 1>}
  NamedRerouteDeclaration(None=Saturate(None=ScalarParameter() {group=FacialExpressionBlend, parameterName=Normal_Curve, sortPriority=1})) {name=Blend, nodeColor=<rgba 0 0.4609 1 1>}
  n13 = NamedRerouteUsage()
  n14 = TextureSampleParameter2D() {
   group=FacialExpressionBlend
   parameterName=FacialExpression_SRM
   samplerType=SAMPLERTYPE_Masks
   sortPriority=31
   texture=&T_Fortnite_Default_S
  }
  n15 = NamedRerouteUsage()
  NamedRerouteDeclaration(None=StaticSwitchParameter(True=LinearInterpolate(A=@n13, B=@n14.RGB, Alpha=@n15), False=NamedRerouteUsage()) {group=FacialExpressionBlend, parameterName=BlendSRM, sortPriority=30}) {name=OUT_SRM, nodeColor=<rgba 1 0 0.7969 1>}
  n16 = NamedRerouteUsage()
  n17 = TextureSampleParameter2D() {
   group=FacialExpressionBlend
   parameterName=FacialExpression_STT
   samplerType=SAMPLERTYPE_Masks
   sortPriority=41
   texture=/Game/Global/Textures/Default/Blanks/T_FN_Anime_STT
  }
  n18 = NamedRerouteUsage()
  NamedRerouteDeclaration(None=StaticSwitchParameter(True=LinearInterpolate(A=@n16, B=@n17.RGB, Alpha=@n18), False=NamedRerouteUsage()) {group=FacialExpressionBlend, parameterName=BlendSTT, sortPriority=40}) {name=OUT_STT, nodeColor=<rgba 1 0 0.7969 1>}
  NamedRerouteDeclaration(None=FunctionInput() {inputName=Color_Shaded, sortPriority=7}) {name=IN_DiffuseShaded, nodeColor=<rgba 0 0.2969 1 1>}
 ]}
